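# Sound cue for the elimination reaction sound of the Origin Prisoner back bling cosmetic.
# A short fixed delay feeds a modulator that boosts volume with slight variation, then a random node picks one of six equally weighted elimination wave variants.
# It uses close pet attenuation, the reactive back bling property sheet and sound class, and a concurrency override limiting it to one instance per owner.
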SoundCue /BRCosmetics/Sounds/Backblings/OriginPrisoner/Backbling_Origin_Prisoner_Elimination_01_Cue
root=Delay_1
nodes [
 {id=Delay_1, children [Modulator_0], properties {delayMax=0.3, delayMin=0.3}}
 {id=Modulator_0, children [Random_1], properties {volumeMax=1.25, volumeMin=1.2}}
 {
  id=Random_1
  children [WavePlayer_0, WavePlayer_1, WavePlayer_2, WavePlayer_9, WavePlayer_10, WavePlayer_11]
  properties {weights [1,1,1,1,1,1]}
 }
 {id=WavePlayer_0, properties {soundWaveAssetPtr=&/Backbling_Origin_Prisoner_Elimination_03_06}}
 {id=WavePlayer_1, properties {soundWaveAssetPtr=&/Backbling_Origin_Prisoner_Elimination_03_01}}
 {id=WavePlayer_10, properties {soundWaveAssetPtr=&/Backbling_Origin_Prisoner_Elimination_03_04}}
 {id=WavePlayer_11, properties {soundWaveAssetPtr=&/Backbling_Origin_Prisoner_Elimination_03_05}}
 {id=WavePlayer_2, properties {soundWaveAssetPtr=&/Backbling_Origin_Prisoner_Elimination_03_02}}
 {id=WavePlayer_9, properties {soundWaveAssetPtr=&/Backbling_Origin_Prisoner_Elimination_03_03}}
]
properties {
 attenuationSettings=&Fort_Pets_Close_Athena_Att
 audioPropertiesSheet=/BRCosmetics/Sounds/Settings/PropertySheets/Reactive/APS_Cosmetic_Reactive_Backbling
 bOverrideConcurrency=true
 maxDistance=1800.0001
 soundClassObject=/Game/Sounds/FortSoundClassesAndModes/Fort_Backblings
 volumeMultiplier=1
 concurrencyOverrides {bLimitToOwner=true, maxCount=1, volumeScale=0.7}
}
